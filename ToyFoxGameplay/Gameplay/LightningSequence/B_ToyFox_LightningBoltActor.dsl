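; Short-lived lightning bolt actor that runs a timed cadence of three strikes at its location, then silently dies.
; Local delays on BeginPlay keep cosmetics and gameplay aligned without sending repeated RPCs to the lobby.
; Each strike async-loads and plays Niagara impact FX and sound where cosmetics can run.
; On authority, it runs an async sphere overlap with a hotfixable radius and applies a knockback gameplay effect to hit actors.
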(event EventBeginPlay
  ;[0 0.0625 0.006971 0.5] BeginPlay: Execute lightning cadence. Using simple delays to keep cosmetic and gameplay lined up for the brief duration this actor is alive instead of sending multiple RPCs over a short duration for the entire lobby.
  (|CallParentFunction)
  (Utilities|FlowControl|Sequence
    (:then_0
      (Utilities|FlowControl|Delay 4.0)
      (Gameplay|SilentDie false))
    (:then_1
      (CallFunction|ExecuteLightning)
      (Utilities|FlowControl|Delay 1.1)
      (CallFunction|ExecuteLightning)
      (Utilities|FlowControl|Delay 0.9)
      (CallFunction|ExecuteLightning))))

(event Custom|ExecuteLightning
  ;[0.1196 0.06611 0.15 0.5] ExecuteLightning: On the server, this performs the knockback and on the client this performs the cosmetic.
  (CallFunction|PlayLightningCosmetic)
  (CallFunction|PerformLightningKnockback))

(event Custom|PlayLightningCosmetic
  ;[0 0.009054 0.15 0.5] Play the lightning cosmetic.
  (Networking|CanExecuteCosmeticEvents
    (:True
      (bind _object (Utilities|AsyncLoadAsset (Variables|FX|GetNiagaraTemplateLightningBoltGroundImpact))
        (:then
          (bind _object_2 (Utilities|AsyncLoadAsset (Variables|FX|GetSoundCueLightningBolt))
            (:Completed
              (bind _assound_base (Utilities|Casting|CastToSoundBase _object_2))
              (Audio|PlaySoundatLocation _assound_base (Transformation|GetActorLocation)))))
        (:Completed
          (bind _returnvalue (Niagara|SpawnSystematLocation (Utilities|Casting|CastToNiagaraSystem _object) (Transformation|GetActorLocation) (Math|Rotator|MakeRotator 0.0 0.0 (Math|Random|RandomFloatinRange 0.0 3.6e+02)) "2.000000,2.000000,2.000000" true true "AutoRelease"))
          (Rendering|SetTranslucentSortPriority _returnvalue 1))))))

(event Custom|PerformLightningKnockback
  ;[0.15 0.01622 0.1371 0.5] PerformLightningKnockback: Find actors around the area and apply knockback.
  (Networking|SwitchHasAuthority
    (:Authority
      (Class|AsyncOverlapQueue|AsyncSphereOverlapActors (Transformation|GetActorLocation) (ScalableFloat|GetCachedScalableFloatValue (Variables|Hotfix|GetHotfix_KnockbackActorsDetectRadius)) (Utilities|Array|MakeArray self) (EventDispatchers|CreateEvent :!event HandleAsyncOverlapSpawnKnocbackShouldCancel) (EventDispatchers|CreateEvent :!event HandleAsyncOverlapSpawnKnockbackHitActor) (EventDispatchers|CreateEvent :!event HandleAsyncOverlapSpawnKnockbackComplete) "ECC_GameTraceChannel3"))))

(fn HandleAsyncOverlapSpawnKnocbackShouldCancel
  (return (Game|IsActorBeingDestroyed)))

(fn HandleAsyncOverlapSpawnKnockbackComplete
  (params (WasCancelled :type Boolean)))

(fn HandleAsyncOverlapSpawnKnockbackHitActor
  (params (HitActor :type Actor%20Object%20Reference))
  ;[0.1283 0.04173 0.15 0.5] Knockback any overlapping actors.
  (Networking|SwitchHasAuthority
    (:Authority
      (Utilities|IsValid HitActor
        (:Is%20Valid
          (Fortnite|ApplyGameplayEffectsToActor HitActor (Utilities|Array|MakeArray (Variables|Settings|GetKnockbackGameplayEffect)) (Transformation|GetActorLocation) false 0 (Ability|GetFortAbilitySystemComponent self) (GameplayEffects|MakeFortEffectContext (Ability|GetFortAbilitySystemComponent self) self self 0 0)))))))

(fn ConstructionScript)
