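; Networked ground rumble actor for the ToyFox lightning sequence that plays a camera shake and audio, then switches them off after a set duration.
; On BeginPlay the server waits for playlist data to be ready, delays by a hotfixable scalable-float duration, sets bDeactivate and flushes net dormancy.
; The bDeactivate OnRep, where cosmetic events can run, deactivates the camera shake source and fades out the audio.
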(event EventBeginPlay
  ;[0 0.15 0.000647 0.5] Deactivate visuals after a certain duration.
  (|CallParentFunction)
  (Networking|SwitchHasAuthority
    (:Authority
      (Class|FortAsyncActionCurrentPlaylistReady|PlaylistDataReadyAsync
        (:OnReady
          (Utilities|FlowControl|Delay (ScalableFloat|GetValueatLevel (Variables|Hotfix|GetHotfix_Duration)))
          (Variables|Transient|SetbDeactivate true)
          (Networking|FlushNetDormancy))))))

(event Custom|OnRep_bDeactivate
  ;[0 0.15 0.000647 0.5] Deactivate visuals after a certain duration.
  (Networking|CanExecuteCosmeticEvents
    (:True
      (if (Variables|Transient|GetbDeactivate)
        (Components|Activation|Deactivate (Variables|Default|GetCameraShakeSource))
        (Audio|Components|Audio|FadeOut (Variables|Default|GetAudio) 0.4 0.0 "Linear" :!class /Script/Engine.AudioComponent)))))

(fn ConstructionScript)
